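; Child balloon actor Blueprint for the LunarEel weapons content, adding a small tweak to the inherited pop effect.
; PlayPopFX hides the static mesh and its children in game, marked as a temporary placeholder fix, then passes the hit result to the parent.
; BeginPlay, overlap and Tick are unused stubs, and the construction script only calls the parent.
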(event EventBeginPlay
  (|CallParentFunction) ; This node is disabled and will not be called.\nDrag off pins to build functionality.
)

(event Collision|EventActorBeginOverlap
  (params (OtherActor :type Actor%20Object%20Reference)))

(event EventTick
  (params (DeltaSeconds :type Float%20%28single-precision%29)))

(event EventPlayPopFX
  (params (HitEvent :type Hit%20Result%20Structure))
  ;[1 0.1346 0 0.5] Temporary call to hide a placeholder
  (Development|SetHiddeninGame (Variables|Default|GetStaticMesh) true true)
  (|CallParentFunction HitEvent))

(fn ConstructionScript
  (|CallParentFunction))
