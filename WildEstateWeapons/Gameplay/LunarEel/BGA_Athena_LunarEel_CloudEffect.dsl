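; Cosmetic effect actor for the Lunar Eel weapon that plays a balloon explosion with a party sound when it spawns.
; After a very brief delay on BeginPlay, it plays the sound at the actor's location.
; It then copies its did-hit-targets flag into the Niagara component's IsHitPlayer parameter and activates the explosion, so the visual can differ on a hit.
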(event EventBeginPlay
  (Utilities|FlowControl|Delay 0.01)
  (Audio|PlaySoundatLocation (Variables|Default|GetPartySoundCue) (Transformation|GetActorLocation))
  (if (Variables|Default|GetdidHitTargets?)
    (Niagara|SetNiagaraVariable(Bool) (Variables|Default|GetNS_LunarEel_Balloon_Explosion) "IsHitPlayer" true)
    (else
      (Niagara|SetNiagaraVariable(Bool) (Variables|Default|GetNS_LunarEel_Balloon_Explosion) "IsHitPlayer" false)))
  (Components|Activation|Activate (Variables|Default|GetNS_LunarEel_Balloon_Explosion) false))

(fn ConstructionScript)
