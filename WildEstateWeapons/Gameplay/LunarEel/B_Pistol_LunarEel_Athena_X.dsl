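; Weapon Blueprint for the exotic Lunar Eel pistol that gives the first shot of each magazine special fire effects.
; On weapon fire FX, it calls the parent, then derives BulletIndex from clip size minus local magazine ammo and flags index zero as FirstShot.
; A first shot spawns the exotic burst Niagara muzzle flash and FirstShotSound on the weapon mesh; other shots spawn the master muzzle flash.
; Both flashes attach at the muzzle socket with ConfettiIndex zeroed.
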(event EventOnPlayWeaponFireFX
  (params (bPersistentFire :type Boolean) (bSecondaryFire :type Boolean))
  (|CallParentFunction bPersistentFire bSecondaryFire)
  (bind _output_get (Variables|Default|SetBulletIndex (- (FortWeapon|GetBulletsPerClip) (FortWeapon|GetLocalMagazineAmmoCount))))
  (bind _output_get_2 (Variables|Default|SetFirstShot (== _output_get 0)))
  (if _output_get_2
    (bind _returnvalue (Niagara|SpawnSystemAttached "/WildEstateWeapons/Assets/FX/LunarEel/NS_LunarEel_MuzzleFlash_Exotic_Burst.NS_LunarEel_MuzzleFlash_Exotic_Burst" (Variables|Visual|GetWeaponMesh) (Effects|GetMuzzleSocketName 0) "0, 0, 0" "0, 0, 0" "SnapToTarget" true))
    (Niagara|SetNiagaraVariable(Int32) _returnvalue "ConfettiIndex" 0)
    (Audio|SpawnSoundAttached (Variables|Default|GetFirstShotSound) (Variables|Visual|GetWeaponMesh))
    (else
      (bind _returnvalue_2 (Niagara|SpawnSystemAttached "/WildEstateWeapons/Assets/FX/Muzzle/NS_MuzzleFlash_Master.NS_MuzzleFlash_Master" (Variables|Visual|GetWeaponMesh) (Effects|GetMuzzleSocketName 0) "0, 0, 0" "0, 0, 0" "SnapToTarget" true))
      (Niagara|SetNiagaraVariable(Int32) _returnvalue_2 "ConfettiIndex" 0))))

(fn ConstructionScript
  (|CallParentFunction))
