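; Gameplay ability for the LunarEel weapon's knockback shot, activated by a hit event and gated on bullet count within the clip.
; Triggered shots launch direct-hit pawns upward or attach balloons to vehicles, then spawn a cloud actor at impact.
; A sphere overlap there grants characters a balloon effect and launches them using hotfixable values, while vehicles receive impulses.
; A delayed velocity check plays the knockback GameplayCue when vertical speed is high.
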(event Custom|FlintKnockback
  (params (CharacterRef :type Character%20Object%20Reference) (WasDirectHit :type Boolean))
  (bind _returnvalue (GameplayTags|HasAnyMatchingGameplayTags(Message) CharacterRef (Variables|Knockback|GetTags_PreventKnockback)))
  (if (not _returnvalue)
    (CallFunction|LaunchPlayer CharacterRef WasDirectHit)
    ;[1 1 1] Wait a brief moment (Like 2 frames) then check velocity. If Velocity on Z is high enough or low enough play GC.
    (Utilities|Time|SetTimerbyEvent (delegate Custom|CheckVel) (Variables|Knockback|GetFX%20Delay))
    (else
      (Ability|EndAbility))))

(event Custom|CheckVel
  (bind _returnvalue (Ability|GetAvatarActorfromActorInfo))
  ;[1 1 1] Wait a brief moment (Like 2 frames) then check velocity. If Velocity on Z is high enough or low enough play GC.
  (Utilities|IsValid _returnvalue
    (:Is%20Valid
      (bind _asfort_player_pawn (Utilities|Casting|CastToFortPlayerPawn _returnvalue))
      (Development|PrintString (Utilities|String|ToString(Vector) (Transformation|GetVelocity _asfort_player_pawn)))
      (bind (_returnvalue_x _returnvalue_y _returnvalue_z) (Math|Vector|BreakVector (Transformation|GetVelocity _asfort_player_pawn)))
      (if (> (Math|Float|Absolute(Float) _returnvalue_z) (Variables|Knockback|GetVelocityForFX))
        (Ability|ExecuteGameplayCueOnOwner (Variables|Default|GetKnockback%20Effect)))
      (Ability|EndAbility))))

(event Custom|TryDeployBalloon
  (params (EffectTargetActor :type Actor%20Object%20Reference))
  ;[1 1 1] Grant a balloon
  (Ability|ApplyGameplayEffectToTarget (Ability|TargetData|AbilityTargetDatafromActor EffectTargetActor) "/WildEstateWeapons/Gameplay/LunarEel/GE_LunarEel_Balloons_Consumable_Passive_A.GE_LunarEel_Balloons_Consumable_Passive_A_C") ; give balloon GE
)

(event Ability|EventActivateAbilityFromEvent
  (params (EventData :type Gameplay%20Event%20Data%20Structure%20%28by%20ref%29))
  (CallFunction|CheckandHandleBulletCount) ; Trigger the effect on X bullet in the clip
  (Variables|Default|SetHitPawn? false)
  (Variables|Default|SetInstigator (Utilities|Struct|BreakGameplayEventData EventData))
  (bind (_valid _location _normal _hitactor) (CallFunction|GetHitInfo EventData))
  (if _valid
    (Default|HitPawnCheck _hitactor
      (:exec
        (Variables|Default|SetHit%20Location _location)
        (Variables|Default|SetHitNormal _normal)
        ;[1 1 1] Trigger the effect only if that was the X
        ; bullet in the clip
        (if (Utilities|ShouldEffectTrigger?)
          (if (Variables|Default|GetHitPawn?)
            (Default|DirectHitLaunchHandle)
            (Utilities|FlowControl|Delay 0.15)
            (else
              (CallFunction|TryAttachBalloontoVehicle _hitactor (Variables|Default|GetHit%20Location))))
          (Default|EnvironmentalKnockback 0 (Variables|Default|GetHitPawn?))
          (Ability|EndAbility)
          (else
            (Ability|EndAbility)))))
    (else
      (Ability|EndAbility))))

(fn CalculateVehicleImpulse
  (params (Object :type Actor%20Object%20Reference))
  (return (Math|Vector|MakeVector (.x (* (Math|Vector|Normalize (- (Transformation|GetActorLocation Object) (Variables|Default|GetHit%20Location))) (ScalableFloat|GetValueatLevel (Variables|Hotfix|GetHF_VehicleKnockbackIntensity)))) (.y (* (Math|Vector|Normalize (- (Transformation|GetActorLocation Object) (Variables|Default|GetHit%20Location))) (ScalableFloat|GetValueatLevel (Variables|Hotfix|GetHF_VehicleKnockbackIntensity)))) (+ (.z (* (Math|Vector|Normalize (- (Transformation|GetActorLocation Object) (Variables|Default|GetHit%20Location))) (ScalableFloat|GetValueatLevel (Variables|Hotfix|GetHF_VehicleKnockbackIntensity)))) (ScalableFloat|GetValueatLevel (Variables|Hotfix|GetHF_VehicleZOffsetKnockbackIntensity))))))

(fn CheckAndHandleBulletCount
  (bind _returnvalue (Ability|GetCurrentSourceWeapon))
  (bind _returnvalue_2 (FortWeapon|Ammo|HasInfiniteMagazineAmmo _returnvalue))
  (bind _returnvalue_3 (not _returnvalue_2))
  (bind _returnvalue_4 (FortWeapon|GetBulletsPerClip _returnvalue))
  ;[1 1 1] If not using the infinite ammo config
  (if _returnvalue_3
    (Variables|DynamicVariables|SetBulletCount (- (FortWeapon|GetBulletsPerClip (Ability|GetCurrentSourceWeapon)) (FortWeapon|GetLocalMagazineAmmoCount (Ability|GetCurrentSourceWeapon))))
    (elif (and _returnvalue_3 (== (+ (FortWeapon|GetLocalMagazineAmmoCount _returnvalue) 1) _returnvalue_4))
      (Variables|DynamicVariables|SetBulletCount 0)
      ;[1 1 1] Check if reached bullets per clip
      (if (>= (Variables|DynamicVariables|GetBulletCount) _returnvalue_4) :!label _label
        (Variables|DynamicVariables|SetBulletCount 1)
        (else
          (bind _result (Math|Integer|IncrementInt (Variables|DynamicVariables|GetBulletCount))
            (:%20%20
              (Variables|DynamicVariables|SetBulletCount _result)))))
      (else
        (goto _label))))
  (return))

(fn DirectHitLaunchHandle
  (Utilities|LaunchPawns (CallFunction|GetAvatarasFortPlayerPawn) (Variables|Default|GetTargetHitPawn) (Math|Vector|MakeVector 0.0 0.0 (ScalableFloat|GetValueatLevel (Variables|Hotfix|GetHF_DirectHitLaunchZStrength))) false true true true (!= (ScalableFloat|GetValueatLevel (Variables|Hotfix|GetHF_ShouldDirectHitLaunchTeammates)) 0) true false false))

(fn EnvironmentalKnockback
  (params (hitPawn? :type Boolean))
  (bind _returnvalue (Building|SpawnBuildingGameplayActor "/WildEstateWeapons/Gameplay/LunarEel/BGA_Athena_LunarEel_CloudEffect.BGA_Athena_LunarEel_CloudEffect_C" (Math|Transform|MakeTransform (Variables|Default|GetHit%20Location) (select hitPawn? "0, 0, 0" (Math|Rotator|MakeRotfromZ (Variables|Default|GetHitNormal)))) (Variables|Default|GetInstigator) true false)) ; Spawn BGA with explosion VFX
  (bind (_outactors _returnvalue_2) (Collision|SphereOverlapActors (Variables|Default|GetHit%20Location) (ScalableFloat|GetValueatLevel (Variables|Hotfix|GetHF_KnockbackRadius)) (Variables|Default|GetVehicleFilterObjectType))) ; See what actors are affected by explosion
  (Variables|/WildEstateWeapons/Gameplay/LunarEel/BGA_Athena_LunarEel_CloudEffect.BGA_Athena_LunarEel_CloudEffect_C|SetdidHitTargets? _returnvalue_2 (Utilities|Casting|CastToBGA_Athena_LunarEel_CloudEffect _returnvalue)) ; If did not hit any targets, change NS in BGA
  (if _returnvalue_2
    (LocalVariables|SetOffset (ScalableFloat|GetValueatLevel (Variables|Hotfix|GetRow_OffsetFromHitNormal)))
    (LocalVariables|SetOffsetCrouched (ScalableFloat|GetValueatLevel (Variables|Hotfix|GetRow_OffsetFromHitNormalCrouched)))
    (for _array_element _outactors
      (if (not (Variables|Default|GetHitPawn?))
        ;[0.02058 0.9167 0.01189 0.5] From surface normal hit
        (bind _returnvalue_3 (Fortnite|CheckLineOfSightToActorWithChannel (+ (Variables|Default|GetHit%20Location) (Utilities|Operators|Multiply (Variables|Default|GetHitNormal) (Math|Vector|MakeVector (Math|Float|SelectFloat (LocalVariables|GetOffsetCrouched) (LocalVariables|GetOffset) (Variables|/Script/Engine.Character|GetbIsCrouched (Utilities|Casting|CastToCharacter _array_element))) (Math|Float|SelectFloat (LocalVariables|GetOffsetCrouched) (LocalVariables|GetOffset) (Variables|/Script/Engine.Character|GetbIsCrouched (Utilities|Casting|CastToCharacter _array_element))) (Math|Float|SelectFloat (LocalVariables|GetOffsetCrouched) (LocalVariables|GetOffset) (Variables|/Script/Engine.Character|GetbIsCrouched (Utilities|Casting|CastToCharacter _array_element)))))) _array_element "ECC_GameTraceChannel7"))
        (if _returnvalue_3
          _))
      (if (.bSuccess (Utilities|Casting|CastToCharacter _array_element))
        (CustomVerb|BroadcastCustomVerbMessage (CustomVerb|MakeCustomVerbMessage (Variables|Default|GetInstigator) "(TagName=\"CustomVerb.ShockwaveTarget\")"))
        (CallFunction|TryDeployBalloon (Utilities|Casting|CastToCharacter _array_element))
        (CallFunction|FlintKnockback (Utilities|Casting|CastToCharacter _array_element) (Variables|Default|GetHitPawn?))
        (else
          (bind _impulsevector (CallFunction|CalculateVehicleImpulse _array_element))
          (Utilities|LaunchVehicle _array_element _impulsevector true))))
    (return)))

(fn GetHitInfo
  (params (EventData :type Gameplay%20Event%20Data%20Structure))
  (bind (_eventtag _instigator _target _optionalobject _optionalobject2 _contexthandle _instigatortags _targettags _eventmagnitude _targetdata) (Utilities|Struct|BreakGameplayEventData EventData))
  (for _index (range (Ability|TargetData|GetDataCountfromTargetData _targetdata))
    (if (Collision|BreakHitResult (Ability|TargetData|GetHitResultfromTargetData _targetdata _index))
      (return true (.Location (Collision|BreakHitResult (Ability|TargetData|GetHitResultfromTargetData _targetdata _index))) (.Normal (Collision|BreakHitResult (Ability|TargetData|GetHitResultfromTargetData _targetdata _index))) (.HitActor (Collision|BreakHitResult (Ability|TargetData|GetHitResultfromTargetData _targetdata _index))))))
  (return false "0, 0, 0" "0, 0, 0" 0))

(fn LaunchPlayer
  (params (CharacterRef :type Character%20Object%20Reference) (WasDirectHit :type Boolean))
  (bind _characterref (Variables|GetCharacterRef))
  (bind _finallaunchvel (Variables|Knockback|GetFinalLaunchVel))
  ;[1 1 1] Cast to pawn and launch the player a little bit. The distance launcged is based on a max distance from the player who hit them. If the player is further than the max distance than no launch is applied. Otherwise the player is launched based on a ratio of the knock mag using the distance from the player who launched them.
  (if (Networking|IsServer)
    (Utilities|IsValid CharacterRef
      (:Is%20Valid
        (bind _asfort_player_pawn (Utilities|Casting|CastToFortPlayerPawn CharacterRef)
          (:then
            ;;[1 1 1] Check for Knock Immune or In Vehicle
            (LocalVariables|SetLocal_FortPlayerPawn _asfort_player_pawn)
            (if (or (Vehicle|IsinVehicle _asfort_player_pawn) (GameplayTags|HasMatchingGameplayTag _asfort_player_pawn "(TagName=\"Gameplay.Status.KnockbackImmunity\")"))
              _
              (else
                _)))
          (:CastFailed
            (Utilities|Casting|CastToCharacter CharacterRef)))
        ;;[1 1 1] If Crouched do different knock, but not if they are falling, if they are falling do the normal one.
        (if (and (Utilities|Enum|NotEqual(Enum) (Variables|/Script/Engine.CharacterMovementComponent|GetMovementMode (Variables|/Script/Engine.Character|GetCharacterMovement (Variables|GetCharacterRef))) "MOVE_Falling") (Variables|/Script/Engine.Character|GetbIsCrouched (Variables|GetCharacterRef)))
          _)
        (bind _local_fortplayerpawn (LocalVariables|GetLocal_FortPlayerPawn)
          (:then
            (FortMovement|CancelJumpAbility _local_fortplayerpawn false))
          (:else
            _))
        ;;[1 1 1] Calc Launch Vel
        (Variables|Knockback|SetFinalLaunchVel (Math|Vector|MakeVector (.x (* (* (Math|Vector|Normalize (Math|Vector|SelectVector (* (Math|Vector|GetForwardVector (Math|Rotator|MakeRotator (.roll (TargetSelection|GetDirectionActortoActor (Variables|Default|GetInstigator) (Variables|GetCharacterRef))) (+ (.pitch (TargetSelection|GetDirectionActortoActor (Variables|Default|GetInstigator) (Variables|GetCharacterRef))) (ScalableFloat|GetValueatLevel (Variables|Hotfix|GetRow_OffsetPitch))) (.yaw (TargetSelection|GetDirectionActortoActor (Variables|Default|GetInstigator) (Variables|GetCharacterRef))))) -1.0) (Math|Vector|GetForwardVector (Math|Rotator|MakeRotator (.roll (Math|Rotator|FindLookatRotation (Transformation|GetActorLocation (Variables|GetCharacterRef)) (Variables|Default|GetHit%20Location))) (+ (.pitch (Math|Rotator|FindLookatRotation (Transformation|GetActorLocation (Variables|GetCharacterRef)) (Variables|Default|GetHit%20Location))) (ScalableFloat|GetValueatLevel (Variables|Hotfix|GetRow_OffsetPitch))) (.yaw (Math|Rotator|FindLookatRotation (Transformation|GetActorLocation (Variables|GetCharacterRef)) (Variables|Default|GetHit%20Location))))) (Variables|GetWasDirectHit))) (Math|Float|SelectFloat (ScalableFloat|GetValueatLevel (Variables|Hotfix|GetRow_KnockMagDirectHit)) (ScalableFloat|GetValueatLevel (Variables|Hotfix|GetRow_KnockMag)) (Variables|GetWasDirectHit))) -1.0)) (.y (* (* (Math|Vector|Normalize (Math|Vector|SelectVector (* (Math|Vector|GetForwardVector (Math|Rotator|MakeRotator (.roll (TargetSelection|GetDirectionActortoActor (Variables|Default|GetInstigator) (Variables|GetCharacterRef))) (+ (.pitch (TargetSelection|GetDirectionActortoActor (Variables|Default|GetInstigator) (Variables|GetCharacterRef))) (ScalableFloat|GetValueatLevel (Variables|Hotfix|GetRow_OffsetPitch))) (.yaw (TargetSelection|GetDirectionActortoActor (Variables|Default|GetInstigator) (Variables|GetCharacterRef))))) -1.0) (Math|Vector|GetForwardVector (Math|Rotator|MakeRotator (.roll (Math|Rotator|FindLookatRotation (Transformation|GetActorLocation (Variables|GetCharacterRef)) (Variables|Default|GetHit%20Location))) (+ (.pitch (Math|Rotator|FindLookatRotation (Transformation|GetActorLocation (Variables|GetCharacterRef)) (Variables|Default|GetHit%20Location))) (ScalableFloat|GetValueatLevel (Variables|Hotfix|GetRow_OffsetPitch))) (.yaw (Math|Rotator|FindLookatRotation (Transformation|GetActorLocation (Variables|GetCharacterRef)) (Variables|Default|GetHit%20Location))))) (Variables|GetWasDirectHit))) (Math|Float|SelectFloat (ScalableFloat|GetValueatLevel (Variables|Hotfix|GetRow_KnockMagDirectHit)) (ScalableFloat|GetValueatLevel (Variables|Hotfix|GetRow_KnockMag)) (Variables|GetWasDirectHit))) -1.0)) (Math|Float|float+float (.z (* (* (Math|Vector|Normalize (Math|Vector|SelectVector (* (Math|Vector|GetForwardVector (Math|Rotator|MakeRotator (.roll (TargetSelection|GetDirectionActortoActor (Variables|Default|GetInstigator) (Variables|GetCharacterRef))) (+ (.pitch (TargetSelection|GetDirectionActortoActor (Variables|Default|GetInstigator) (Variables|GetCharacterRef))) (ScalableFloat|GetValueatLevel (Variables|Hotfix|GetRow_OffsetPitch))) (.yaw (TargetSelection|GetDirectionActortoActor (Variables|Default|GetInstigator) (Variables|GetCharacterRef))))) -1.0) (Math|Vector|GetForwardVector (Math|Rotator|MakeRotator (.roll (Math|Rotator|FindLookatRotation (Transformation|GetActorLocation (Variables|GetCharacterRef)) (Variables|Default|GetHit%20Location))) (+ (.pitch (Math|Rotator|FindLookatRotation (Transformation|GetActorLocation (Variables|GetCharacterRef)) (Variables|Default|GetHit%20Location))) (ScalableFloat|GetValueatLevel (Variables|Hotfix|GetRow_OffsetPitch))) (.yaw (Math|Rotator|FindLookatRotation (Transformation|GetActorLocation (Variables|GetCharacterRef)) (Variables|Default|GetHit%20Location))))) (Variables|GetWasDirectHit))) (Math|Float|SelectFloat (ScalableFloat|GetValueatLevel (Variables|Hotfix|GetRow_KnockMagDirectHit)) (ScalableFloat|GetValueatLevel (Variables|Hotfix|GetRow_KnockMag)) (Variables|GetWasDirectHit))) -1.0)) (Math|Float|SelectFloat (ScalableFloat|GetValueatLevel (Variables|Hotfix|GetRow_ZAddDirectHit)) (ScalableFloat|GetValueatLevel (Variables|Hotfix|GetRow_ZAdd)) (Variables|GetWasDirectHit)) (* (.z (Transformation|GetVelocity (Variables|GetCharacterRef))) (ScalableFloat|GetValueatLevel (Variables|Hotfix|GetRow_ZGravMult))))))
        ;;[1 1 1] Stop Movement and launch player
        (Components|Movement|StopMovementImmediately (Variables|/Script/Engine.Character|GetCharacterMovement _characterref))
        (Character|LaunchCharacter _characterref _finallaunchvel true true)
        ;;[1 1 1] If Velocity is Pos Z reset fall height
        (bind _local_fortplayerpawn_2 (LocalVariables|GetLocal_FortPlayerPawn)
          (:then
            (if (and (Math|Conversions|ToBoolean(Integer) (Math|Float|Floor (ScalableFloat|GetValueatLevel (Variables|Hotfix|GetRow_ResetFall)))) (> (.z _finallaunchvel) 0.0))
              (Pawn|Character|ResetFallingHeight _local_fortplayerpawn_2)))
          (:else))))))

(fn TryAttachBalloonToVehicle
  (params (HitActor :type Actor%20Object%20Reference) (HitLocation :type Vector))
  (if (ScalableFloat|GetValueasBool (Variables|Hotfix|GetHF_AttachToVehicles?))
    (bind _asfort_athena_vehicle (Utilities|Casting|CastToFortAthenaVehicle HitActor))
    ;[1 1 1] Don't attach a balloon the vehicle already has one
    (Utilities|IsValid (Actor|GetComponentbyClass _asfort_athena_vehicle "/Script/FortniteGame.FortPhysicsConstantForceComponent")
      (:Is%20Valid
        (if (Components|Activation|IsActive (Actor|GetComponentbyClass _asfort_athena_vehicle "/Script/FortniteGame.FortPhysicsConstantForceComponent"))
          _
          (else
            _)))
      (:Is%20Not%20Valid
        _))
    (Game|SpawnActorfromClass "/WildEstateWeapons/Gameplay/LunarEel/PhysicsBalloon/BGA_PhysicsBalloons_Async.BGA_PhysicsBalloons_Async_C" (Math|Transform|MakeTransform HitLocation) "Undefined" "MultiplyWithRoot" 0 _asfort_athena_vehicle "None" 0 0 0 0 0)))
