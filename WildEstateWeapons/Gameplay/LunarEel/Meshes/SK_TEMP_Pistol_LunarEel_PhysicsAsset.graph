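# Physics asset for the temporary LunarEel pistol skeletal mesh, giving the weapon a simplified collision shape.
# It holds a single body on the Handle bone, made of one rotated, offset capsule.
# It has no constraints, so the body has no jointed or articulated parts.
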PhysicsAsset /WildEstateWeapons/Gameplay/LunarEel/Meshes/SK_TEMP_Pistol_LunarEel_PhysicsAsset
bodies {Handle [{shapeType=Capsule, length=16.4355, radius=9.3256, center=<8.8923 -0.2423 4.8068>, rotation=<rot 0 90.4874 -67.0993>}]}
constraints []
