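# Layered sound cue for the Lunar Eel party popper explosion, using a shared attenuation asset and a concurrency override.
# One submix plays a random popper sound plus two delayed random detail layers, the second having empty slots that sometimes leave it silent.
# The other submix combines random balloon explosion and VO clips with a slightly delayed balloon inflate clip.
# Both submixes feed the root mixer through modulators and weighted mixing for per-play variation.
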SoundCue /WildEstateWeapons/Gameplay/LunarEel/Sound/Cue_PartyPopper_LunarEel
root=Mixer_2
nodes [
 {id=Delay_0, children [Modulator_2], properties {delayMax=0.075, delayMin=0.075}}
 {id=Delay_1, children [Modulator_5], properties {delayMax=0.075, delayMin=0.075}}
 {id=Mixer_0, children [Modulator_1,Delay_0], properties {inputVolume [1,1]}}
 {id=Mixer_1, children [Modulator_4,Random_6,Delay_1], properties {inputVolume [1.5,1,1.25]}}
 {id=Mixer_2, children [Mixer_0,Mixer_1], properties {inputVolume [0.8,0.8]}}
 {id=Mixer_3, children [Random_2,Random_1], properties {inputVolume [1,1.25]}}
 {id=Modulator_1, children [Random_0]}
 {id=Modulator_2, children [Mixer_3]}
 {id=Modulator_4, children [Random_4], properties {volumeMin=0.98}}
 {id=Modulator_5, children [Random_5]}
 {id=Random_0, children [WavePlayer_8, WavePlayer_9, WavePlayer_14, WavePlayer_15], properties {weights [1,1,1,1]}}
 {id=Random_1, children [WavePlayer_4, WavePlayer_6, WavePlayer_7, WavePlayer_10, null, null], properties {weights [0.8,0.8,0.8,0.8,1,1]}}
 {id=Random_2, children [WavePlayer_0,WavePlayer_1,WavePlayer_2,WavePlayer_3], properties {weights [1,1,1,1]}}
 {id=Random_4, children [WavePlayer_5, WavePlayer_16, WavePlayer_17], properties {weights [1,1,1]}}
 {id=Random_5, children [WavePlayer_18, WavePlayer_19, WavePlayer_20], properties {weights [1,1,1]}}
 {
  id=Random_6
  children [WavePlayer_11, WavePlayer_12, WavePlayer_13, WavePlayer_21, WavePlayer_22, WavePlayer_23]
  properties {weights [1,1,1,1,1,1]}
 }
 {id=WavePlayer_0, properties {soundWaveAssetPtr=/WildEstateWeapons/Gameplay/LunarEel/Sound/Waves/Explo_LunarEel_PartyPopper_Lyr_01}}
 {id=WavePlayer_1, properties {soundWaveAssetPtr=/WildEstateWeapons/Gameplay/LunarEel/Sound/Waves/Explo_LunarEel_PartyPopper_Lyr_02}}
 {id=WavePlayer_10, properties {soundWaveAssetPtr=/WildEstateWeapons/Gameplay/LunarEel/Sound/Waves/Explo_LunarEel_PartyPopper_Lyr_05}}
 {id=WavePlayer_11, properties {soundWaveAssetPtr=/WildEstateWeapons/Gameplay/LunarEel/Sound/Waves/Explo_LunarEel_VO_01}}
 {id=WavePlayer_12, properties {soundWaveAssetPtr=/WildEstateWeapons/Gameplay/LunarEel/Sound/Waves/Explo_LunarEel_VO_02}}
 {id=WavePlayer_13, properties {soundWaveAssetPtr=/WildEstateWeapons/Gameplay/LunarEel/Sound/Waves/Explo_LunarEel_VO_03}}
 {id=WavePlayer_14, properties {soundWaveAssetPtr=&/S_PartyPopper_02}}
 {id=WavePlayer_15, properties {soundWaveAssetPtr=&/S_PartyPopper_01}}
 {id=WavePlayer_16, properties {soundWaveAssetPtr=/WildEstateWeapons/Gameplay/LunarEel/Sound/Waves/Explo_LunarEel_Balloon_Explosion_02}}
 {id=WavePlayer_17, properties {soundWaveAssetPtr=/WildEstateWeapons/Gameplay/LunarEel/Sound/Waves/Explo_LunarEel_Balloon_Explosion_03}}
 {id=WavePlayer_18, properties {soundWaveAssetPtr=/WildEstateWeapons/Gameplay/LunarEel/Sound/Waves/Explo_LunarEel_Balloon_Inflate_01}}
 {id=WavePlayer_19, properties {soundWaveAssetPtr=/WildEstateWeapons/Gameplay/LunarEel/Sound/Waves/Explo_LunarEel_Balloon_Inflate_02}}
 {id=WavePlayer_2, properties {soundWaveAssetPtr=/WildEstateWeapons/Gameplay/LunarEel/Sound/Waves/Explo_LunarEel_PartyPopper_Lyr_03}}
 {id=WavePlayer_20, properties {soundWaveAssetPtr=/WildEstateWeapons/Gameplay/LunarEel/Sound/Waves/Explo_LunarEel_Balloon_Inflate_03}}
 {id=WavePlayer_21, properties {soundWaveAssetPtr=/WildEstateWeapons/Gameplay/LunarEel/Sound/Waves/Explo_LunarEel_VO_04}}
 {id=WavePlayer_22, properties {soundWaveAssetPtr=/WildEstateWeapons/Gameplay/LunarEel/Sound/Waves/Explo_LunarEel_VO_05}}
 {id=WavePlayer_23, properties {soundWaveAssetPtr=/WildEstateWeapons/Gameplay/LunarEel/Sound/Waves/Explo_LunarEel_VO_06}}
 {id=WavePlayer_3, properties {soundWaveAssetPtr=/WildEstateWeapons/Gameplay/LunarEel/Sound/Waves/Explo_LunarEel_PartyPopper_Lyr_04}}
 {id=WavePlayer_4, properties {soundWaveAssetPtr=/WildEstateWeapons/Gameplay/LunarEel/Sound/Waves/Explo_LunarEel_PartyPopper_Lyr_08}}
 {id=WavePlayer_5, properties {soundWaveAssetPtr=/WildEstateWeapons/Gameplay/LunarEel/Sound/Waves/Explo_LunarEel_Balloon_Explosion_01}}
 {id=WavePlayer_6, properties {soundWaveAssetPtr=/WildEstateWeapons/Gameplay/LunarEel/Sound/Waves/Explo_LunarEel_PartyPopper_Lyr_07}}
 {id=WavePlayer_7, properties {soundWaveAssetPtr=/WildEstateWeapons/Gameplay/LunarEel/Sound/Waves/Explo_LunarEel_PartyPopper_Lyr_06}}
 {id=WavePlayer_8, properties {soundWaveAssetPtr=&/S_PartyPopper_04}}
 {id=WavePlayer_9, properties {soundWaveAssetPtr=&/S_PartyPopper_03}}
]
properties {
 attenuationSettings=&/LunarEel_PartyPop_Explosion_Att
 bDebug=true
 bOverrideConcurrency=true
 maxDistance=7000
 volumeMultiplier=1
 attenuationOverrides {falloffDistance=5600}
 concurrencyOverrides {maxCount=5}
}
