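# Sound cue for the LunarEel PizzaParty explosion, layering a balloon explosion with a balloon inflate sound.
# A Mixer blends two Modulator branches, each fed by a Random node choosing one of three variant waves, one set for explosions and one for inflation.
# It uses RPG explosion attenuation, the gadgets sound class, and a concurrency override that prevents new instances once the limit is reached.
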SoundCue /WildEstateWeapons/Gameplay/LunarEel/Sound/LunarEel_PizzaParty_Explosion_Cue
root=Mixer_0
nodes [
 {id=Mixer_0, children [Modulator_0,Modulator_1], properties {inputVolume [1,1]}}
 {id=Modulator_0, children [Random_2]}
 {id=Modulator_1, children [Random_0]}
 {id=Random_0, children [WavePlayer_6,WavePlayer_7,WavePlayer_8], properties {weights [1,1,1]}}
 {id=Random_2, children [WavePlayer_3,WavePlayer_4,WavePlayer_5], properties {weights [1,1,1]}}
 {id=WavePlayer_3, properties {soundWaveAssetPtr=/WildEstateWeapons/Gameplay/LunarEel/Sound/Waves/Explo_LunarEel_Balloon_Explosion_01}}
 {id=WavePlayer_4, properties {soundWaveAssetPtr=/WildEstateWeapons/Gameplay/LunarEel/Sound/Waves/Explo_LunarEel_Balloon_Explosion_02}}
 {id=WavePlayer_5, properties {soundWaveAssetPtr=/WildEstateWeapons/Gameplay/LunarEel/Sound/Waves/Explo_LunarEel_Balloon_Explosion_03}}
 {id=WavePlayer_6, properties {soundWaveAssetPtr=/WildEstateWeapons/Gameplay/LunarEel/Sound/Waves/Explo_LunarEel_Balloon_Inflate_01}}
 {id=WavePlayer_7, properties {soundWaveAssetPtr=/WildEstateWeapons/Gameplay/LunarEel/Sound/Waves/Explo_LunarEel_Balloon_Inflate_02}}
 {id=WavePlayer_8, properties {soundWaveAssetPtr=/WildEstateWeapons/Gameplay/LunarEel/Sound/Waves/Explo_LunarEel_Balloon_Inflate_03}}
]
properties {
 attenuationSettings=/Game/Sounds/Attenuation/Fort_Explosion_RPG_Att
 bOverrideConcurrency=true
 maxDistance=13000
 soundClassObject=&Fort_Gadgets
 volumeMultiplier=1
 concurrencyOverrides {resolutionRule=PreventNew}
}
